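# Niagara effect for the MagmaBreak backpack cosmetic, with flame meshes, glow sprites and curl-noise embers on skeletal mesh points.
# System.Timeline.Time drives the brief ember spike and the flame and glow envelopes; an Eliminations input suggests kill reactivity.
# Four flames and glows select attachments and offsets by spawn group; flames pass smoothed positions to materials.
# Time-of-day tints all emitters, and front-end modifiers dim them.
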NiagaraSystem /BRCosmetics/Accessories/FORT_Backpacks/Backpack_M_MED_MagmaBreak/FX/NS_BackPack_MagmaBreak_Idle
dependencies {
 usedDataInterfaces [
  &NiagaraDataInterfaceCurve
  &NiagaraDataInterfaceVectorCurve
  &NiagaraDataInterfaceArrayFloat
  &NiagaraDataInterfaceArrayFloat3
  &NiagaraDataInterfaceArrayFloat2
  &NiagaraDataInterfaceArrayInt32
  &NiagaraDataInterfaceColorCurve
  &NiagaraDataInterfaceSkeletalMesh
 ]
 usedDynamicInputs [
  &MakeCustomFloatFromBool
  &SetBoolByIntComparison
  &Divide_Float
 ]
 usedModules [
  &self:SystemSpawnScript.SystemScriptSource.SystemScriptGraph.NiagaraNodeAssignment_0.SetVariables_#1
  &self:SystemSpawnScript.SystemScriptSource.SystemScriptGraph.NiagaraNodeAssignment_2.SetVariables_#2
  &SystemState
  /Game/Effects/Niagara/Modules/Logic/UpdateValueStateChange
  &self:SystemSpawnScript.SystemScriptSource.SystemScriptGraph.NiagaraNodeAssignment_3.SetVariables_#3
  &Timeline
  &self:SystemSpawnScript.SystemScriptSource.SystemScriptGraph.NiagaraNodeAssignment_1.SetVariables_#4
 ]
}
emitters {
 Embers {
  emitterUpdateScript [
   &EmitterState()
   &SpawnRate(
    SpawnRate=&FloatFromCurve(
     FloatCurve=DataInterface() {propertyValues {Curve {keys [<at 0.0367 0>,<at 0.0393 1>,<at 0.05 0>]}}}
     CurveIndex=$System.Timeline.Time
     Scale Curve=222
    )
    Use Spawn Probability=false
   )
  ]
  particleSpawnScript [
   &InitializeParticle(
    Lifetime Max=1.2
    Lifetime Min=0.9
    Lifetime Mode=::NewEnumerator1
    Sprite Rotation Mode=::NewEnumerator2
    Sprite Size Mode=::NewEnumerator1
    Uniform Sprite Size Max=4
    Uniform Sprite Size Min=1.5
    UsePositionOffset=false
    Write Material Random=true
   )
   &AddVelocity(Velocity=&RandomRangeVector(Minimum=<-30 0 0>, Maximum=<30 -50 30>))
   &SkeletalMeshLocation(
    Bone Sampling Mode=::NewEnumerator4
    MaskSpawnGroup=false
    Sampled Position Offset=&RandomRangeVector(Minimum=<-3 -3 -3>, Maximum=<3 3 3>)
    Skeletal Mesh=$System.SkeletalMesh
   )
  ]
  particleUpdateScript [
   &Drag(Drag=1.5, Use Linear Drag=true, Use Rotational Drag=true) {enabled=false}
   &CurlNoiseForce(
    Noise Frequency=9
    Noise Quality / Cost=::NewEnumerator0
    Noise Strength=&RandomRangeFloat(Minimum=46.9161, Maximum=131.0828)
    Pan Noise=true
    Pan Noise Field=&RandomRangeVector(Minimum=<0 0 0>, Maximum=<1 1 1>)
    Random Seed=2
   )
   &InheritSourceMovement(
    Applied Position Scale=&VectorFromCurve(VectorCurve=DataInterface() {propertyValues {
       XCurve {keys [<at 0 1 RCIM_Cubic>,<at 0.3 0>]}
       YCurve {keys [<at 0 1 RCIM_Cubic>,<at 0.3 0>]}
       ZCurve {keys [<at 0 1 RCIM_Cubic>,<at 0.3 0>]}
      }})
    ApplyForce=false
    ApplyPosition=true
    ApplyPositionFalloff=false
   ) {enabled=false}
   &SolveForcesAndVelocity(Clamp Velocity=false, Limit Acceleration=false)
   &ParticleState()
   &ScaleSpriteSizeBySpeed(Max Scale Factor=<1.2 4>, Min Scale Factor=<0.9 1.4>)
   &ScaleSpriteSize(Uniform Curve Sprite Scale=DataInterface() {propertyValues {Curve {keys [
        <at 0 0 RCIM_Cubic>
        {arriveTangent=1.5714, interpMode=RCIM_Cubic, leaveTangent=1.5714, time=0.4, value=1}
        {arriveTangent=-1.0833, interpMode=RCIM_Cubic, leaveTangent=-1.0833, time=0.7, value=1.1}
        <at 1 0.35 RCIM_Cubic>
       ]}}})
   /SetVariables(Particles.Color=&MultiplyLinearColor(
     Linear Color=$Particles.Initial.Color
     Scale Factor=&MakeVector4_FromLinearColor(LinearColor=&FNTimeOfDayColor(
       Night Color=<rgba 0.277 0.2868 0.4792 0.85>
       Morning Color=<rgba 0.8906 0.5659 0.5659 1>
       Day Color=<rgba 1 1 1 1>
       Evening Color=<rgba 1 1 1 1>
      ))
    )) {isSetParameters=true}
   &ScaleColor(
    Scale Alpha=&FloatFromCurve(
     FloatCurve=DataInterface() {propertyValues {Curve {keys [<at 0 1>, {interpMode=RCIM_Cubic, leaveTangent=-4.1367, tangentMode=RCTM_Break, time=0.65, value=0.95}, <at 1 0>]}}}
     Scale Curve=&FN_FrontEndFloatModifier(Gameplay=1, Front End=0.7)
    )
    ScaleA=true
    ScaleRGB=true
   )
   &DynamicMaterialParameters(
    Index 0 Param 1=&FloatFromCurve(FloatCurve=DataInterface() {propertyValues {Curve {keys [
         <at -0.1 0.05>
         {interpMode=RCIM_Cubic, leaveTangent=1.2819, tangentMode=RCTM_Break, time=0.4, value=0.2}
         {arriveTangent=0.0217, tangentMode=RCTM_User, time=0.7, value=0.95}
        ]}}})
    Index 0 Param 2=&RandomRangeFloat(Minimum=0, Maximum=1)
    Index 0 Param 3=&RandomRangeFloat(Minimum=0, Maximum=1)
    Module.Index0Param0Write=true
    Module.Index0Param1Write=true
    Module.Index0Param2Write=true
    Module.Index0Param3Write=true
   )
  ]
  rendererClasses [&NiagaraSpriteRendererProperties]
  renderers [{rendererClass=&NiagaraSpriteRendererProperties}]
 }
 Flames_Spline_Mesh {
  emitterUpdateScript [
   &EmitterState(
    Inactive Response=::NewEnumerator1
    Life Cycle Mode=::NewEnumerator1
    Loop Behavior=::NewEnumerator1
    Loop Duration Mode=::NewEnumerator1
    UseLoopDelay=false
   )
   &SpawnBurst_Instantaneous(Spawn Count=4, Use Spawn Probability=false, UseLoopCountLimit=false)
  ]
  particleSpawnScript [
   &FN_SetSpawnGroup()
   &InitializeParticle(
    Color=<rgba 0.0469 0.0469 0.0469 1>
    Lifetime=&SelectFloatFromArray(Direct Array Index=&SpawnGroup(), Float Selection Array=DataInterface() {propertyValues {FloatData [7,6]}})
    Mesh Scale=&SelectVectorFromArray(
     Direct Array Index=&SpawnGroup()
     Vector Selection Array=DataInterface() {propertyValues {FloatData [<0.33 0.5 0.5>, <0.1 0.4 0.4>]}}
    )
    Mesh Scale Mode=::NewEnumerator4
    UsePositionOffset=false
    Write Material Random=true
   )
   &InitialMeshOrientation(
    Mesh Orientation Mode=::NewEnumerator2
    Orientation Vector=&SelectVectorFromArray(
     Direct Array Index=&SpawnGroup()
     Vector Selection Array=DataInterface() {propertyValues {FloatData [<0 -1 0>, <0 -1 0>, <0 -1 0>, <0 -1 0>]}}
    )
    Rotation=&SelectVectorFromArray(
     Direct Array Index=&SpawnGroup()
     Vector Selection Array=DataInterface() {propertyValues {FloatData [<0 0 0>, <0 0 0>, <0.03 0 0>, <-0.03 0 0>]}}
    )
    Use Rotation Vector=true
   )
   /SetVariables(Particles.GlowScale:Vector3f=&Multiply_Vector(A=$Particles.Scale, B=<1.3 1.3 1.3>)) {isSetParameters=true}
  ]
  particleUpdateScript [
   &ParticleState(Kill Particles When Lifetime Has Elapsed=false, Let Infinitely Lived Particles Die When Emitter Deactivates=true)
   &SkeletalMeshLocation(
    Bone / Socket Index=&SpawnGroup()
    Bone Sampling Mode=::NewEnumerator4
    MaskSpawnGroup=false
    Sampled Position Offset=&SelectVectorFromArray(
     Direct Array Index=&SpawnGroup()
     Vector Selection Array=DataInterface() {propertyValues {FloatData [<0 -4.65 -0.5>, <0 1 -0.3>, <0 1 0>, <0 1 0>]}}
    )
    Skeletal Mesh=$System.SkeletalMesh
   ) {module=SkeletalMeshLocation001}
   &InterpolateOverTime(Rate Of Change=15, Target Vector Value:NiagaraPosition=$Particles.Position, Variable Type=::NewEnumerator2) {module=InterpolateOverTime001}
   &DynamicMaterialParameters(
    Index 0 Param 1=&Multiply_Float(
     A=&Lerp_Float(
      A=&FloatFromCurve(
       FloatCurve=DataInterface() {propertyValues {Curve {keys [
           <at 0.05 0.6 RCIM_Cubic>
           <at 0.14 1>
           {arriveTangent=-0.4651, interpMode=RCIM_Cubic, leaveTangent=-0.4651, time=0.8, value=0.95}
           <at 1 0.6>
          ]}}}
       CurveIndex=$System.Timeline.Time
      )
      B=&FloatFromCurve(
       FloatCurve=DataInterface() {propertyValues {Curve {keys [
           <at 0.05 0.6 RCIM_Cubic>
           <at 0.14 1>
           {arriveTangent=-0.5405, interpMode=RCIM_Cubic, leaveTangent=-0.5405, time=0.72, value=0.95}
           <at 0.88 0.6>
          ]}}}
       CurveIndex=$System.Timeline.Time
      )
      Alpha=&ClampFloat(Float=&MakeFloatFromInt(Int=&SpawnGroup()))
     )
     B=&SelectFloatFromArray(Direct Array Index=&SpawnGroup(), Float Selection Array=DataInterface() {propertyValues {FloatData [1.4,1]}})
    )
    Index 0 Param 2=&Multiply_Float(
     A=&Lerp_Float(
      A=&FloatFromCurve(
       FloatCurve=DataInterface() {propertyValues {Curve {keys [<at 0.05 0.6 RCIM_Cubic>, <at 0.1 1>, {arriveTangent=-0.4444, interpMode=RCIM_Cubic, leaveTangent=-0.4444, time=0.8, value=0.95}, <at 1 0.6>]}}}
       CurveIndex=$System.Timeline.Time
      )
      B=&FloatFromCurve(
       FloatCurve=DataInterface() {propertyValues {Curve {keys [
           <at 0.05 0.6 RCIM_Cubic>
           <at 0.1 1>
           {arriveTangent=-0.5128, interpMode=RCIM_Cubic, leaveTangent=-0.5128, time=0.6594, value=0.95}
           <at 0.88 0.6>
          ]}}}
       CurveIndex=$System.Timeline.Time
      )
      Alpha=&ClampFloat(Float=&MakeFloatFromInt(Int=&SpawnGroup()))
     )
     B=&SelectFloatFromArray(Direct Array Index=&SpawnGroup(), Float Selection Array=DataInterface() {propertyValues {FloatData [2.5,1.8]}})
    )
    Index 0 Param 3=1
    Index 0 Param 4=0.81
    Index 1 Param 1=&MakeFloatFromVector(Vector=$StackContext.InterpolateOverTime001.InterpolateOverTime_V3_.Moving Average, Channel=::NewEnumerator0)
    Index 1 Param 2=&MakeFloatFromVector(Vector=$StackContext.InterpolateOverTime001.InterpolateOverTime_V3_.Moving Average, Channel=::NewEnumerator1)
    Index 1 Param 3=&MakeFloatFromVector(Vector=$StackContext.InterpolateOverTime001.InterpolateOverTime_V3_.Moving Average, Channel=::NewEnumerator2)
    Index 2 Param 1=&RandomRangeFloat(Minimum=0, Maximum=1)
    Module.Index0Param0Write=true
    Module.Index0Param1Write=true
    Module.Index0Param2Write=true
    Module.Index0Param3Write=true
    Module.Index1Param0Write=true
    Module.Index1Param1Write=true
    Module.Index1Param2Write=true
    Module.Index1Param3Write=true
    Module.Index2Param0Write=true
    Module.Index2Param1Write=true
    Module.Index2Param2Write=true
    Module.Index2Param3Write=true
    Write Parameter Index 1=true
    Write Parameter Index 2=true
   )
   /SetVariables(
    Particles.Dynamic Parameter Dark:Vector4f=&MakeVector4(
     X=&Multiply_Float(
      A=&Lerp_Float(
       A=&FloatFromCurve(
        FloatCurve=DataInterface() {propertyValues {Curve {keys [
            <at 0.05 0.6 RCIM_Cubic>
            <at 0.14 1>
            {arriveTangent=-0.4651, interpMode=RCIM_Cubic, leaveTangent=-0.4651, time=0.8, value=0.95}
            <at 1 0.6>
           ]}}}
        CurveIndex=$System.Timeline.Time
       )
       B=&FloatFromCurve(
        FloatCurve=DataInterface() {propertyValues {Curve {keys [
            <at 0.05 0.6 RCIM_Cubic>
            <at 0.14 1>
            {arriveTangent=-0.5405, interpMode=RCIM_Cubic, leaveTangent=-0.5405, time=0.7149, value=0.95}
            <at 0.88 0.6>
           ]}}}
        CurveIndex=$System.Timeline.Time
       )
       Alpha=&ClampFloat(Float=&MakeFloatFromInt(Int=&SpawnGroup()))
      )
      B=&SelectFloatFromArray(Direct Array Index=&SpawnGroup(), Float Selection Array=DataInterface() {propertyValues {FloatData [4,3]}})
     )
     Y=&Multiply_Float(
      A=&Lerp_Float(
       A=&FloatFromCurve(
        FloatCurve=DataInterface() {propertyValues {Curve {keys [
            <at 0.05 0.6 RCIM_Cubic>
            <at 0.14 1>
            {arriveTangent=-0.4651, interpMode=RCIM_Cubic, leaveTangent=-0.4651, time=0.8, value=0.95}
            <at 1 0.6>
           ]}}}
        CurveIndex=$System.Timeline.Time
       )
       B=&FloatFromCurve(
        FloatCurve=DataInterface() {propertyValues {Curve {keys [
            <at 0.05 0.6 RCIM_Cubic>
            <at 0.14 1>
            {arriveTangent=-0.5405, interpMode=RCIM_Cubic, leaveTangent=-0.5405, time=0.7178, value=0.95}
            <at 0.88 0.6>
           ]}}}
        CurveIndex=$System.Timeline.Time
       )
       Alpha=&ClampFloat(Float=&MakeFloatFromInt(Int=&SpawnGroup()))
      )
      B=&SelectFloatFromArray(Direct Array Index=&SpawnGroup(), Float Selection Array=DataInterface() {propertyValues {FloatData [1.2,1.2]}})
     )
     Z=1
     W=1.06
    )
    Particles.Dynamic Parameter Dark1:Vector4f=&MakeVector4(X=&RandomRangeFloat(Minimum=0, Maximum=1), Y=1, Z=1, W=1)
   ) {isSetParameters=true}
   /SetVariables(Particles.Inner Color:LinearColor=&MultiplyLinearColor(
     Linear Color=<rgba 2.6 0.3428 0.0406 0.739>
     Scale Factor=&MakeVector4_FromFloat(Float=&Multiply_Float(
       A=&FNTimeOfDayScalar(Night Scale=0.66, Morning Scale=0.85, Evening Scale=0.8)
       B=&FloatFromCurve(
        FloatCurve=DataInterface() {propertyValues {Curve {keys [<at 0.05 0 RCIM_Cubic>, <at 0.14 1>, {arriveTangent=-1.1628, interpMode=RCIM_Cubic, leaveTangent=-1.1628, time=0.8, value=0.95}, <at 1 0>]}}}
        CurveIndex=$System.Timeline.Time
       )
      ))
    )) {isSetParameters=true}
   /SetVariables(Particles.Color=&MultiplyLinearColor(
     Linear Color=<rgba 0.7292 0.0738 0.003 0.613>
     Scale Factor=&MakeVector4_FromFloat(Float=&Multiply_Float(
       A=&FNTimeOfDayScalar(Night Scale=0.66, Morning Scale=0.85, Evening Scale=0.8)
       B=&FloatFromCurve(
        FloatCurve=DataInterface() {propertyValues {Curve {keys [<at 0.05 0 RCIM_Cubic>, <at 0.14 1>, {arriveTangent=-1.1628, interpMode=RCIM_Cubic, leaveTangent=-1.1628, time=0.8, value=0.95}, <at 1 0>]}}}
        CurveIndex=$System.Timeline.Time
       )
      ))
    )) {isSetParameters=true}
   /SetVariables(Particles.GlowColor=&MultiplyLinearColor(
     Linear Color=<rgba 3 0.4486 0 0.396>
     Scale Factor=&MakeVector4_FromFloat(Float=&Multiply_Float(
       A=&FNTimeOfDayScalar(Night Scale=0.66, Morning Scale=0.85, Evening Scale=0.8)
       B=&FloatFromCurve(
        FloatCurve=DataInterface() {propertyValues {Curve {keys [<at 0.05 0 RCIM_Cubic>, <at 0.14 1>, {arriveTangent=-1.1628, interpMode=RCIM_Cubic, leaveTangent=-1.1628, time=0.8, value=0.95}, <at 1 0>]}}}
        CurveIndex=$System.Timeline.Time
        Scale Curve=&FN_FrontEndFloatModifier(Gameplay=1, Front End=0.55)
       )
      ))
    )) {isSetParameters=true}
   /SetVariables(Particles.Dynamic Parameter Glow:Vector4f=&MakeVector4(
     X=&Multiply_Float(
      A=&FloatFromCurve(
       FloatCurve=DataInterface() {propertyValues {Curve {keys [
           <at 0.05 0.6 RCIM_Cubic>
           <at 0.14 1>
           {arriveTangent=-0.4651, interpMode=RCIM_Cubic, leaveTangent=-0.4651, time=0.8, value=0.95}
           <at 1 0.6>
          ]}}}
       CurveIndex=$System.Timeline.Time
      )
      B=5
     )
     Y=&Multiply_Float(
      A=&FloatFromCurve(
       FloatCurve=DataInterface() {propertyValues {Curve {keys [
           <at 0.05 0.6 RCIM_Cubic>
           <at 0.14 1>
           {arriveTangent=-0.4651, interpMode=RCIM_Cubic, leaveTangent=-0.4651, time=0.8, value=0.95}
           <at 1 0.6>
          ]}}}
       CurveIndex=$System.Timeline.Time
      )
      B=2
     )
     Z=1
     W=1.06
    )) {isSetParameters=true}
   &ScaleMeshSize(Scale Factor=&VectorFromFloat(Value=&Lerp_Float(
      A=&FloatFromCurve(
       FloatCurve=DataInterface() {propertyValues {Curve {keys [
           <at 0.0583 0.5 RCIM_Cubic>
           {arriveTangent=2.6904, interpMode=RCIM_Cubic, leaveTangent=2.6904, time=0.1299, value=1.3}
           {interpMode=RCIM_Cubic, tangentMode=RCTM_User, time=0.2441, value=1}
           {arriveTangent=-0.6615, interpMode=RCIM_Cubic, leaveTangent=-0.6615, time=0.8085, value=0.95}
           {arriveTangent=-3.4093, tangentMode=RCTM_User, time=1, value=0.5}
          ]}}}
       CurveIndex=$System.Timeline.Time
      )
      B=&FloatFromCurve(
       FloatCurve=DataInterface() {propertyValues {Curve {keys [
           <at 0.0583 0.5 RCIM_Cubic>
           {arriveTangent=2.6904, interpMode=RCIM_Cubic, leaveTangent=2.6904, time=0.1299, value=1.3}
           {interpMode=RCIM_Cubic, tangentMode=RCTM_User, time=0.2441, value=1}
           {arriveTangent=-0.7863, interpMode=RCIM_Cubic, leaveTangent=-0.7863, time=0.7154, value=0.95}
           <at 0.88 0.5>
          ]}}}
       CurveIndex=$System.Timeline.Time
      )
      Alpha=&ClampFloat(Float=&MakeFloatFromInt(Int=&SpawnGroup()))
     )))
  ]
  rendererClasses [&NiagaraMeshRendererProperties]
  renderers [
   {rendererClass=&NiagaraMeshRendererProperties, rendererIndex=0}
   {rendererClass=&NiagaraMeshRendererProperties, rendererIndex=1}
   {rendererClass=&NiagaraMeshRendererProperties, rendererIndex=2}
  ]
 }
 Glow {
  emitterUpdateScript [
   &EmitterState(
    Inactive Response=::NewEnumerator1
    Life Cycle Mode=::NewEnumerator1
    Loop Behavior=::NewEnumerator1
    Loop Duration Mode=::NewEnumerator1
    UseLoopDelay=false
   )
   &SpawnBurst_Instantaneous(Spawn Count=4, Use Spawn Probability=false, UseLoopCountLimit=false)
  ]
  particleSpawnScript [
   &FN_SetSpawnGroup()
   &InitializeParticle(
    Color=<rgba 1 0.3325 0.004 1>
    Lifetime=&SelectFloatFromArray(Direct Array Index=&SpawnGroup(), Float Selection Array=DataInterface() {propertyValues {FloatData [7,6]}})
    Sprite Size=&SelectVector2DFromArray(
     Direct Array Index=&SpawnGroup()
     Vector 2D Selection Array=DataInterface() {propertyValues {FloatData [<25 25>, <15 15>, <15 15>, <15 15>]}}
    )
    Sprite Size Mode=::NewEnumerator4
    UsePositionOffset=false
    Write Material Random=true
   )
   &CameraOffset(Camera Offset Amount=&SelectFloatFromArray(Direct Array Index=&SpawnGroup(), Float Selection Array=DataInterface() {propertyValues {FloatData [5,5,5,2]}}))
  ]
  particleUpdateScript [
   &ParticleState(Kill Particles When Lifetime Has Elapsed=false, Let Infinitely Lived Particles Die When Emitter Deactivates=true)
   &ScaleSpriteSize(
    Uniform Curve Index=$System.Timeline.Time
    Uniform Curve Sprite Scale=DataInterface() {propertyValues {Curve {keys [
        <at 0.05 0 RCIM_Cubic>
        {interpMode=RCIM_Cubic, tangentMode=RCTM_User, time=0.0707, value=1}
        {arriveTangent=0.2317, interpMode=RCIM_Cubic, leaveTangent=0.2317, tangentMode=RCTM_User, time=0.9681, value=1}
        {tangentMode=RCTM_User, time=1, value=0}
       ]}}}
   )
   &SkeletalMeshLocation(
    Bone / Socket Index=&SelectIntFromArray(Direct Array Index=&SpawnGroup(), Int Selection Array=DataInterface() {propertyValues {IntData [0,1,2,3]}})
    Bone Sampling Mode=::NewEnumerator4
    MaskSpawnGroup=false
    Sampled Position Offset=&Add_Vector(
     A=&SelectVectorFromArray(
      Direct Array Index=&SpawnGroup()
      Vector Selection Array=DataInterface() {propertyValues {FloatData [<0 -5 0>, <0 0 0>, <0 0 0>, <0 0 0>]}}
     )
     B=<0 0 0>
    )
    Skeletal Mesh=$System.SkeletalMesh
   )
   &ScaleColor(
    Scale Alpha=&Lerp_Float(
     A=&FloatFromCurve(
      FloatCurve=DataInterface() {propertyValues {Curve {keys [
          {interpMode=RCIM_Cubic, tangentMode=RCTM_User, time=0.08, value=0}
          {arriveTangent=2.8409, interpMode=RCIM_Cubic, leaveTangent=2.8409, time=0.1487, value=0.8147}
          {arriveTangent=-0.6553, interpMode=RCIM_Cubic, leaveTangent=-0.6553, time=0.3, value=0.625}
          {arriveTangent=1.2224, interpMode=RCIM_Cubic, leaveTangent=1.2224, tangentMode=RCTM_User, time=0.4, value=0.65}
          {arriveTangent=1.1515, interpMode=RCIM_Cubic, leaveTangent=1.1515, time=0.4386, value=0.767}
          {interpMode=RCIM_Cubic, tangentMode=RCTM_User, time=0.5185, value=0.7864}
          {arriveTangent=-0.469, interpMode=RCIM_Cubic, leaveTangent=-0.469, time=0.6286, value=0.578}
          {arriveTangent=0.0711, interpMode=RCIM_Cubic, leaveTangent=0.0711, time=0.7672, value=0.6698}
          {arriveTangent=-0.5932, interpMode=RCIM_Cubic, leaveTangent=-0.5932, time=0.8293, value=0.5922}
          {arriveTangent=-2.0999, interpMode=RCIM_Cubic, leaveTangent=-2.0999, time=0.8979, value=0.5922}
          {arriveTangent=-5.8029, interpMode=RCIM_Cubic, leaveTangent=-5.8029, time=0.9548, value=0.3286}
          {arriveTangent=-2.6765, tangentMode=RCTM_User, time=1, value=0}
         ]}}}
      CurveIndex=$System.Timeline.Time
      Scale Curve=&FN_FrontEndFloatModifier(Gameplay=0.33, Front End=0.13)
     )
     B=&FloatFromCurve(
      FloatCurve=DataInterface() {propertyValues {Curve {keys [
          {interpMode=RCIM_Cubic, tangentMode=RCTM_User, time=0.08, value=0}
          {arriveTangent=2.8409, interpMode=RCIM_Cubic, leaveTangent=2.8409, time=0.1273, value=0.8147}
          {arriveTangent=-0.6038, interpMode=RCIM_Cubic, leaveTangent=-0.6038, time=0.3, value=0.625}
          {arriveTangent=1.2224, interpMode=RCIM_Cubic, leaveTangent=1.2224, tangentMode=RCTM_User, time=0.4, value=0.65}
          {arriveTangent=1.1515, interpMode=RCIM_Cubic, leaveTangent=1.1515, time=0.4386, value=0.767}
          {interpMode=RCIM_Cubic, tangentMode=RCTM_User, time=0.5185, value=0.7864}
          {arriveTangent=-0.6144, interpMode=RCIM_Cubic, leaveTangent=-0.6144, time=0.6286, value=0.578}
          {arriveTangent=0.1135, interpMode=RCIM_Cubic, leaveTangent=0.1135, time=0.7083, value=0.6698}
          {arriveTangent=-0.6925, interpMode=RCIM_Cubic, leaveTangent=-0.6925, time=0.7543, value=0.5922}
          {arriveTangent=-4.065, interpMode=RCIM_Cubic, leaveTangent=-4.065, time=0.8203, value=0.5922}
          {tangentMode=RCTM_User, time=0.9, value=0}
         ]}}}
      CurveIndex=$System.Timeline.Time
      Scale Curve=&FN_FrontEndFloatModifier(Gameplay=0.33, Front End=0.13)
     )
     Alpha=&ClampFloat(Float=&MakeFloatFromInt(Int=&SpawnGroup()))
    )
    Scale RGB=&VectorFromFloat(Value=&FNTimeOfDayScalar(Night Scale=0.66, Morning Scale=0.85, Evening Scale=0.83))
    ScaleA=true
    ScaleRGB=true
   )
  ]
  rendererClasses [&NiagaraSpriteRendererProperties]
  renderers [{rendererClass=&NiagaraSpriteRendererProperties}]
 }
}
summary {
 systemName=NS_BackPack_MagmaBreak_Idle
 emitters [
  {emitterName=Embers, rendererClasses [&NiagaraSpriteRendererProperties]}
  {emitterName=Flames_Spline_Mesh, rendererClasses [&NiagaraMeshRendererProperties]}
  {emitterName=Glow, rendererClasses [&NiagaraSpriteRendererProperties]}
 ]
 userVariables [
  User.bIsFrontEnd {type=NiagaraBool}
  User.bIsFrontEndPreview {type=NiagaraBool}
  User.DebugReactivity {type=NiagaraBool}
  User.Eliminations {type=NiagaraInt32}
 ]
}
