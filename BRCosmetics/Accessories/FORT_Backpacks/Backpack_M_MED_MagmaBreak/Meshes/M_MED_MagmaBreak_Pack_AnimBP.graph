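# Animation Blueprint for the MagmaBreak backpack cosmetic that shifts the pack with spine bend and makes its pieces float.
# Each update maps spine_05 yaw relative to spine_01 to pack translation and roll, though the pack bone node leaves rotation mode at default Ignore.
# Real-time sine waves drive offsets for six pieces and the center ball, which also wobbles in rotation.
# The AnimGraph applies these through chained additive bone-space Modify Bone nodes.
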AnimBlueprint /BRCosmetics/Accessories/FORT_Backpacks/Backpack_M_MED_MagmaBreak/Meshes/M_MED_MagmaBreak_Pack_AnimBP
target_skeleton=&/M_MED_MagmaBreak_Pack_Skeleton
graphs [
 AnimGraph {nodes [
   n1 = "Transform (Modify) Bone - Bone: pack"(ComponentPose=Local To Component(LocalPose=Input Pose()), Translation="0,9.3,0", Rotation="0,0,0") {
    translationMode=BMM_Additive
    translationSpace=BCS_BoneSpace
    alphaBoolBlend {blendOption=Linear}
    boneToModify {boneName=pack}
    translation {y=9.3}
   }
   n2 = "Transform (Modify) Bone - Bone: pack"(ComponentPose=@n1, Translation=Get backpack_translation(), Rotation=Get backpack_rotate()) {translationMode=BMM_Additive, translationSpace=BCS_BoneSpace, alphaBoolBlend {blendOption=Linear}, boneToModify {boneName=pack}}
   n3 = "Transform (Modify) Bone - Bone: large_piece_left"(ComponentPose=@n2, Alpha=1, Translation=Get dyn_Large_Left_Float() -> (dyn_Large_Left_Float:Vector), Rotation="0,0,0", Scale="1,1,1") {
    rotationMode=BMM_Additive
    rotationSpace=BCS_BoneSpace
    translationMode=BMM_Additive
    translationSpace=BCS_BoneSpace
    alphaBoolBlend {blendOption=Linear}
    boneToModify {boneName=large_piece_left}
   }
   n4 = "Transform (Modify) Bone - Bone: large_piece_low"(ComponentPose=@n3, Alpha=0.7, Translation=Get dyn_Large_Low_Float() -> (dyn_Large_Low_Float:Vector), Rotation="0,0,0", Scale="1,1,1") {
    rotationMode=BMM_Additive
    rotationSpace=BCS_BoneSpace
    translationMode=BMM_Additive
    translationSpace=BCS_BoneSpace
    alphaBoolBlend {blendOption=Linear}
    boneToModify {boneName=large_piece_low}
   }
   n5 = "Transform (Modify) Bone - Bone: large_piece_right"(ComponentPose=@n4, Alpha=0.8, Translation=Get dyn_Large_Right_Float() -> (dyn_Large_Right_Float:Vector), Rotation="0,0,0", Scale="1,1,1") {
    rotationMode=BMM_Additive
    rotationSpace=BCS_BoneSpace
    translationMode=BMM_Additive
    translationSpace=BCS_BoneSpace
    alphaBoolBlend {blendOption=Linear}
    boneToModify {boneName=large_piece_right}
   }
   n6 = "Transform (Modify) Bone - Bone: center_ball"(
    ComponentPose=@n5
    Alpha=1
    Translation=Get dyn_ball_center_FloatMovem() -> (dyn_ball_center_FloatMovem:Vector)
    Rotation=Get dyn_ball_center_FloatRot() -> (dyn_ball_center_FloatRot:Rotator)
    Scale="1,1,1"
   ) {
    rotationMode=BMM_Additive
    rotationSpace=BCS_BoneSpace
    translationMode=BMM_Additive
    translationSpace=BCS_BoneSpace
    alphaBoolBlend {blendOption=Linear}
    boneToModify {boneName=center_ball}
   }
   n7 = "Transform (Modify) Bone - Bone: small_piece_left"(ComponentPose=@n6, Alpha=1, Translation=Get dyn_Small_Left_Float() -> (dyn_Small_Left_Float:Vector), Rotation="0,0,0", Scale="1,1,1") {
    rotationMode=BMM_Additive
    rotationSpace=BCS_BoneSpace
    translationMode=BMM_Additive
    translationSpace=BCS_BoneSpace
    alphaBoolBlend {blendOption=Linear}
    boneToModify {boneName=small_piece_left}
   }
   n8 = "Transform (Modify) Bone - Bone: small_piece_right"(ComponentPose=@n7, Alpha=1, Translation=Get dyn_Small_Right_Float() -> (dyn_Small_Right_Float:Vector), Rotation="0,0,0", Scale="1,1,1") {
    rotationMode=BMM_Additive
    rotationSpace=BCS_BoneSpace
    translationMode=BMM_Additive
    translationSpace=BCS_BoneSpace
    alphaBoolBlend {blendOption=Linear}
    boneToModify {boneName=small_piece_right}
   }
   n9 = "Transform (Modify) Bone - Bone: small_piece_up"(ComponentPose=@n8, Alpha=1, Translation=Get dyn_Small_Up_Float() -> (dyn_Small_Up_Float:Vector), Rotation="0,0,0", Scale="1,1,1") {
    rotationMode=BMM_Additive
    rotationSpace=BCS_BoneSpace
    translationMode=BMM_Additive
    translationSpace=BCS_BoneSpace
    alphaBoolBlend {blendOption=Linear}
    boneToModify {boneName=small_piece_up}
   }
   Output Pose(Result=Component To Local(ComponentPose=@n9)) {layerGroup=None}
  ]}
 EventGraph {nodes [
   Event BlueprintUpdateAnimation()
   > n1 = Sequence!(execute) -> (then_0, then_1, then_2:Exec, then_3:Exec)
   n2 = GetOwnerSkeletalMeshComponent()
   n3 = BreakRotator(InRot=BreakTransform(InTransform=GetDeltaTransformFromRefPose(self=@n2, BoneName=spine_05, BaseName=spine_01)).Rotation)
   > Set backpack_rotate(backpack_rotate=MakeRotator(Roll=MapRangeClamped(Value=@n3.Yaw, InRangeA=0, InRangeB=-30, OutRangeB=35)))
   > Set backpack_translation(backpack_translation=MakeVector(Y=MapRangeClamped(Value=@n3.Yaw, InRangeA=0, InRangeB=-35, OutRangeA=-4, OutRangeB=-6)))
   n4 = GetRealTimeSeconds()
   n5 = "Sin (Radians)"(A=float * float(A=-4, B=@n4))
   n6 = "Sin (Radians)"(A=float * float(A=1.5, B=@n4))
   Set dyn_Small_Left_Float(execute=@n1.then_3, dyn_Small_Left_Float:Vector=MakeVector(X=float * float(A=@n6, B=0.5), Y=0, Z=float * float(A=@n5, B=0.6))) -> (Output_Get:Vector)
   n7 = GetRealTimeSeconds()
   n8 = "Sin (Radians)"(A=float * float(A=-4, B=@n7))
   n9 = "Sin (Radians)"(A=float * float(A=1.5, B=@n7))
   > Set dyn_Small_Right_Float(dyn_Small_Right_Float:Vector=MakeVector(X=float * float(A=@n9, B=0.25), Y=0, Z=float * float(A=@n8, B=0.6))) -> (Output_Get:Vector)
   n10 = GetRealTimeSeconds()
   n11 = "Sin (Radians)"(A=float * float(A=-4, B=@n10))
   n12 = "Sin (Radians)"(A=float * float(A=1.5, B=@n10))
   > Set dyn_Small_Up_Float(dyn_Small_Up_Float:Vector=MakeVector(X=float * float(A=@n12, B=0.25), Y=0, Z=float * float(A=@n11, B=0.6))) -> (Output_Get:Vector)
   n13 = GetRealTimeSeconds()
   n14 = "Sin (Radians)"(A=float * float(A=1, B=@n13))
   n15 = "Sin (Radians)"(A=float * float(A=1, B=@n13))
   n16 = "Sin (Radians)"(A=float * float(A=2, B=@n13))
   Set dyn_ball_center_FloatRot(
    execute=@n1.then_2
    dyn_ball_center_FloatRot:Rotator=MakeRotator(Roll=float * float(A=@n16, B=4), Pitch=float * float(A=@n15, B=4), Yaw=float * float(A=@n14, B=4))
   ) -> (Output_Get:Rotator)
   n17 = GetRealTimeSeconds()
   n18 = "Sin (Radians)"(A=float * float(A=1, B=@n17))
   n19 = "Sin (Radians)"(A=float * float(A=2, B=@n17))
   > Set dyn_ball_center_FloatMovem(dyn_ball_center_FloatMovem:Vector=MakeVector(X=float * float(A=@n18, B=0.3), Y=0, Z=float * float(A=@n19, B=0.3))) -> (Output_Get:Vector)
   n20 = GetRealTimeSeconds()
   n21 = "Sin (Radians)"(A=float * float(A=4, B=@n20))
   n22 = "Sin (Radians)"(A=float * float(A=1.5, B=@n20))
   Set dyn_Large_Left_Float(execute=@n1.then_1, dyn_Large_Left_Float:Vector=MakeVector(X=float * float(A=@n22, B=0.5), Y=0, Z=float * float(A=@n21, B=0.5))) -> (Output_Get:Vector)
   n23 = GetRealTimeSeconds()
   n24 = "Sin (Radians)"(A=float * float(A=-1.5, B=@n23))
   n25 = "Sin (Radians)"(A=float * float(A=4, B=@n23))
   > Set dyn_Large_Right_Float(dyn_Large_Right_Float:Vector=MakeVector(X=float * float(A=@n24, B=0.5), Y=0, Z=float * float(A=@n25, B=0.5))) -> (Output_Get:Vector)
   n26 = GetRealTimeSeconds()
   n27 = float * float(A="Sin (Radians)"(A=float * float(A=4, B=@n26)), B=0.5)
   n28 = float * float(A="Sin (Radians)"(A=float * float(A=1.5, B=@n26)), B=0.5)
   > Set dyn_Large_Low_Float(dyn_Large_Low_Float:Vector=vector + vector(A=MakeVector(X=@n28, Y=0, Z=@n27), B="0,0,1.5")) -> (Output_Get:Vector)
  ]}
]
variables [
 backpack_rotate {type=Rotator}
 backpack_translation {type=Vector}
 dyn_ball_center_FloatMovem {type=Vector}
 dyn_ball_center_FloatRot {type=Rotator}
 dyn_Large_Left_Float {type=Vector}
 dyn_Large_Low_Float {type=Vector}
 dyn_Large_Right_Float {type=Vector}
 dyn_Small_Left_Float {type=Vector}
 dyn_Small_Right_Float {type=Vector}
 dyn_Small_Up_Float {type=Vector}
]
